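; Camera modifier for the Chain Whip that shifts the view along its forward axis during attacks, with positive offsets pulling it back.
; GA_ChainWhip_Mantis adds to a goal offset, which the current offset chases at constant speed while both settle toward zero.
; Each update applies the alpha-scaled offset to view location only and prints debug values.
; DisableAndRemove fades the modifier out, then removes it from the player camera manager.
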(event Custom|DisableAndRemove
  ;[0.1354 0.08267 0.01257] DisableAndRemove: Disables the modifier and after the alpha out duration, removes it from the camera.
  (CameraModifier|DisableModifier false)
  (Utilities|FlowControl|Delay (Variables|CameraModifier|GetAlphaOutTime))
  (bind _cameraowner (Variables|CameraModifier|GetCameraOwner)
    (:then
      (Utilities|IsValid (Game|Player|GetOwningPlayerController _cameraowner)
        (:Is%20Valid
          (bind _playercameramanager (Variables|/Script/Engine.PlayerController|GetPlayerCameraManager (Game|Player|GetOwningPlayerController _cameraowner))
            (:then
              (Game|Player|RemoveCameraModifier _playercameramanager self))
            (:else)))))
    (:else)))

(fn AddCameraOffset
  (params (Offset :type Float%20%28double-precision%29))
  ;[0.05149 0.2109 0.2292] AddCameraOffset: Called to modify the camera offset when the player is attacking.
  ; Invoked from GA_ChainWhip_Mantis.
  (Variables|Transient|SetGoalOffset (+ Offset (Variables|Transient|GetGoalOffset))))

(fn BlueprintModifyCamera
  (params (DeltaTime :type Float%20%28single-precision%29) (ViewLocation :type Vector :default "0, 0, 0") (ViewRotation :type Rotator :default "0, 0, 0") (FOV :type Float%20%28single-precision%29))
  ;[0.03928 0.1719 0.1667] ModifyCamera: Calculate the offset and apply it to the view location.
  (bind _output_get (Variables|Transient|SetCurrentOffset (Math|Interpolation|FInterptoConstant (Math|Interpolation|FInterptoConstant (Variables|Transient|GetCurrentOffset) (Variables|Transient|GetGoalOffset) DeltaTime 2e+02) 0.0 DeltaTime 5.0)))
  (LocalVariables|SetRetViewLocation (+ (Utilities|Operators|Multiply (Math|Vector|GetForwardVector (Variables|GetViewRotation)) _output_get -1.0 (Variables|CameraModifier|GetAlpha)) ViewLocation))
  (bind _output_get_2 (Variables|Transient|SetGoalOffset (Math|Interpolation|FInterpTo (Variables|Transient|GetGoalOffset) 0.0 DeltaTime 2.0))) ; The goal offset must go towards 0 over time.
  (Development|PrintString (Utilities|String|Append "ChainWhip: CameraModifier: CurrentOffset: " (Utilities|String|ToString(Float) (Variables|Transient|GetCurrentOffset)) " : GoalOffset: " (Utilities|String|ToString(Float) _output_get_2)) true false "(R=0.000000,G=0.660000,B=1.000000,A=1.000000)" 0.0)
  (return (LocalVariables|GetRetViewLocation) (Variables|GetViewRotation) (Variables|GetFOV)))
