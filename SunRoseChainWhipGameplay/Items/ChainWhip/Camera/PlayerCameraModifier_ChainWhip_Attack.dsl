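; Player camera modifier for the chain whip that slows view rotation, apparently during melee attacks.
; When added to a camera manager, it caches a hotfixable scalable-float multiplier into a local variable.
; Each view rotation pass leaves the current rotation unchanged and scales the input rotation delta by that cached multiplier.
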(event EventBlueprintAddedToCamera
  (params (Camera :type Player%20Camera%20Manager%20Object%20Reference))
  ;[0.02248 0.1198 0.04006] AddedToCamera: Cache the multiplier.
  (|CallParentFunction Camera)
  (Variables|Default|SetMeleeAttackViewRotationMultiplier (ScalableFloat|GetValueatLevel (Variables|Default|GetHotfix_MeleeAttackViewRotationMultiplier))))

(fn BlueprintProcessViewRotation
  (params (ViewTarget :type Actor%20Object%20Reference) (DeltaTime :type Float%20%28single-precision%29) (InViewRotation :type Rotator :default "0, 0, 0") (InDeltaRot :type Rotator :default "0, 0, 0"))
  ;[0.04947 0.1354 0.1022] Slow down the rotation rate.
  (return InViewRotation (* InDeltaRot (Variables|Default|GetMeleeAttackViewRotationMultiplier)) true))
